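; Season 28 narrative finale quest actor that animates a Peely mannequin and a hologram character scene, revealed shortly after a delayed start.
; A timeline drives Peely's horizontal offset, height and roll, while the hologram follows a scaled-down share of his frame-to-frame height change.
; The height also toggles the water particle's InWater flag around a threshold.
; On finish, the hologram resets and the motion replays in an endless loop.
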(event EventBeginPlay
  (Rendering|SetVisibility (Variables|Default|GetDefaultSceneRoot) false true)
  (Utilities|FlowControl|Delay 5.0)
  (CallFunction|MovetoNextLocation)
  (Utilities|FlowControl|Delay 0.5)
  (Rendering|SetVisibility (Variables|Default|GetDefaultSceneRoot) true true))

(event Custom|MoveToNextLocation
  (bind (_returnvalue_x _returnvalue_y _returnvalue_z) (Math|Vector|BreakVector (Transformation|GetWorldLocation (Variables|Default|GetMannequin%20Peely))))
  (bind (_returnvalue_x_2 _returnvalue_y_2 _returnvalue_z_2) (Math|Vector|BreakVector (Transformation|GetWorldLocation (Variables|Default|GetMannequin%20Peely))))
  (bind (_returnvalue_roll _returnvalue_pitch _returnvalue_yaw) (Math|Rotator|BreakRotator (Transformation|GetWorldRotation (Variables|Default|GetHolo%20CharacterScene))))
  (bind (_returnvalue_x_3 _returnvalue_y_3 _returnvalue_z_3) (Math|Vector|BreakVector (Transformation|GetWorldLocation (Variables|Default|GetHolo%20CharacterScene))))
  (bind (_returnvalue_x_4 _returnvalue_y_4 _returnvalue_z_4) (Math|Vector|BreakVector (Transformation|GetWorldLocation (Variables|Default|GetMannequin%20Peely))))
  (bind (_holostartingloc_x _holostartingloc_y) (Math|Vector|BreakVector (Variables|GetHoloStartingLoc_X)))
  (bind (_holostartingloc_x_2 _holostartingloc_y_2) (Math|Vector|BreakVector (Variables|GetHoloStartingLoc_X)))
  (Variables|Default|SetMainStartingLoc (Transformation|GetWorldLocation (Variables|Default|GetMannequin%20Peely)))
  (Variables|Default|SetHoloStartingLoc (Transformation|GetWorldLocation (Variables|Default|GetHolo%20CharacterScene)))
  (bind (_direction _heightpath _horizontalpath _rollrotation) (Timeline|Timeline PlayFromStart)
    (:Update
      (CallFunction|HandleVfx _heightpath)
      (Transformation|SetWorldLocation (Variables|Default|GetMannequin%20Peely) (Math|Vector|MakeVector (+ _horizontalpath _holostartingloc_x) _holostartingloc_y _heightpath) false false)
      (Transformation|SetWorldRotation (Variables|Default|GetMannequin%20Peely) (Math|Rotator|MakeRotator _rollrotation _returnvalue_pitch _returnvalue_yaw) false false)
      (if (not (Variables|Default|GetStoredZLocation))
        (Variables|Default|SetStoredZLocation true)
        (Variables|Default|SetMainCurrentStarting _returnvalue_z_2))
      (bind _output_get (Variables|Default|SetDifferenceBetweenZ (- _returnvalue_z (Variables|Default|GetMainCurrentStarting))))
      (Transformation|SetWorldLocation (Variables|Default|GetHolo%20CharacterScene) (Math|Vector|MakeVector _holostartingloc_x_2 _holostartingloc_y_2 (+ (* _output_get 0.3) _returnvalue_z_3)) false false)
      (Variables|Default|SetMainCurrentStarting _returnvalue_z_4))
    (:Finished
      (CallFunction|ResetValues))))

(event Custom|HandleVfx
  (params (NewParam :type Float%20%28double-precision%29))
  (bind _ns_vehicle_meatball_water (Variables|Default|GetNS_Vehicle_Meatball_Water))
  (Math|Float|CompareFloat NewParam -6e+01
    (:>
      (Effects|Components|ParticleSystem|SetBoolParameter _ns_vehicle_meatball_water "InWater" false))
    (:==
      _)
    (:<
      _))
  (Effects|Components|ParticleSystem|SetBoolParameter _ns_vehicle_meatball_water "InWater" true))

(event Custom|ResetValues
  (Variables|Default|SetStoredZLocation false)
  (Transformation|SetWorldLocation (Variables|Default|GetHolo%20CharacterScene) (Variables|Default|GetHoloStartingLoc) false false)
  (Utilities|FlowControl|Delay 0.2)
  (CallFunction|MovetoNextLocation))

(fn ConstructionScript)
