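# Masked, Default Lit material for a map icon among the Season 28 narrative quest actors.
# Its graph holds a single texture sample of the Banana commando (HID-349) hero icon from BRCosmetics, with no connection to any material output shown.
# Despite the MI_ prefix, it is a base material rather than an instance, and it exposes no parameters.
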Material /BattlePassS28_Quests/Items/QuestActors/S28_Narrative/MI_S28_Narrative_Map_Icon
graph {
 properties {blendMode=BLEND_Masked, pixelDepthOffsetMode=PDOM_Legacy, usedShadingModels=Default Lit}
 nodes [TextureSample() {texture=/BRCosmetics/UI/Foundation/Textures/Icons/Heroes/Athena/Soldier/T-Soldier-HID-349-Athena-Commando-M-Banana}]
}
