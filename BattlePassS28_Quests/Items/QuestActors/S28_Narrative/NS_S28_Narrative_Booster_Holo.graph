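# Niagara system for a hologram-named booster effect in the Season 28 narrative quest actors, drawing blaster and thruster mesh particles.
# One continuously spawning emitter of short-lived particles feeds two mesh renderers through separate _Blaster and _Thruster sets of color, size, orientation and dynamic material parameters.
# Curves animate color and size over particle life, the blaster also shrinks late in the emitter loop, and User.uPhase scales both shapes.
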NiagaraSystem /BattlePassS28_Quests/Items/QuestActors/S28_Narrative/NS_S28_Narrative_Booster_Holo
dependencies {
 usedDataInterfaces [&NiagaraDataInterfaceVectorCurve, &NiagaraDataInterfaceCurve, &NiagaraDataInterfaceColorCurve]
 usedDynamicInputs [&ClampFloat, &RemapRange, &VectorLength]
 usedModules [
  &SystemState
  &InterpolateOverTime
  &self:SystemSpawnScript.SystemScriptSource.SystemScriptGraph.NiagaraNodeAssignment_0.SetVariables_#1
  &self:SystemSpawnScript.SystemScriptSource.SystemScriptGraph.NiagaraNodeAssignment_1.SetVariables_#2
 ]
}
emitters {Combined_Rate {
  emitterUpdateScript [
   &EmitterState()
   &SpawnRate(SpawnRate=12, Use Spawn Probability=false)
  ]
  particleSpawnScript [
   &InitializeParticle(
    Color Mode=::NewEnumerator0
    Lifetime Max=0.5
    Lifetime Min=0.25
    Lifetime Mode=::NewEnumerator1
    Position Mode=::NewEnumerator0
    Write Material Random=true
   )
   /SetVariables(
    Particles.Color_Blaster:LinearColor=<rgba 0 1.6667 10 0.1>
    Particles.Orientation_Blaster:Quat4f=&OrientMeshToVector(Look At Direction=<-1 0 0>, Reference Axis Vector=<1 0 0>)
    Particles.Size_Blaster:Vector3f=<1.25 1 1>
   ) {isSetParameters=true}
   /SetVariables(
    Particles.DynamicParameter_Thruster:Vector4f=&MakeVector4(
     X=&RandomRangeFloat(Minimum=0, Maximum=1)
     Y=0
     Z=0
     W=&Lerp_Float(A=10, B=20, Alpha=0.5)
    )
    Particles.Orientation_Thruster:Quat4f=&OrientMeshToVector(Look At Direction=<0 0 1>, Reference Axis Vector=<1 0 0>)
    Particles.Size_Thruster:Vector3f=&RandomRangeVector(Minimum=<0.05 0.05 0.5>, Maximum=<0.125 0.125 0.75>)
   ) {isSetParameters=true}
  ]
  particleUpdateScript [
   /SetVariables(
    Particles.Color_Blaster:LinearColor=&MultiplyLinearColor(
     Linear Color=$Particles.Initial.Color_Blaster
     Scale Factor=&MakeFloat4FromVectorAndFloat(
      "Vector (RGB)"=&VectorFromCurve(VectorCurve=DataInterface() {propertyValues {
         XCurve {keys [<at 0 2 RCIM_Cubic>,<at 1 1 RCIM_Cubic>]}
         YCurve {keys [<at 0 2 RCIM_Cubic>,<at 1 1 RCIM_Cubic>]}
         ZCurve {keys [<at 0 2 RCIM_Cubic>,<at 1 1 RCIM_Cubic>]}
        }})
      "Float (Alpha)"=&FloatFromCurve(FloatCurve=DataInterface() {propertyValues {Curve {keys [<at 0 0 RCIM_Cubic>,<at 0.5 1>,<at 1.0011 0 RCIM_Cubic>]}}})
     )
    )
    Particles.DynamicParameter_Blaster:Vector4f=&MakeVector4(
     X=&FloatFromCurve(FloatCurve=DataInterface() {propertyValues {Curve {keys [<at 0 1>,<at 1 0>]}}})
     Y=5
     Z=20
     W=1
    )
    Particles.Size_Blaster:Vector3f=&Multiply_Vector(
     A=&VectorFromCurve(
      VectorCurve=DataInterface() {propertyValues {
        XCurve {keys [
          <at 0 0.75 RCIM_Cubic>
          <at 0.4908 3.9579>
          {arriveTangent=4.355, interpMode=RCIM_Cubic, leaveTangent=4.355, tangentMode=RCTM_User, time=0.5, value=4}
          <at 1 0.5 RCIM_Cubic>
         ]}
        YCurve {keys [
          <at 0 1 RCIM_Cubic>
          <at 0.4908 0.9933>
          {arriveTangent=0.75, interpMode=RCIM_Cubic, leaveTangent=0.75, time=0.5, value=1}
          <at 1 1.75 RCIM_Cubic>
         ]}
        ZCurve {keys [<at -0.0014 1 RCIM_Cubic>, {arriveTangent=0.7489, interpMode=RCIM_Cubic, leaveTangent=0.7489, time=0.5, value=1}, <at 1 1.75 RCIM_Cubic>]}
       }}
      Scale Curve=&MakeVector(
       X=&Lerp_Float(A=0, B=1, Alpha=$User.uPhase)
       Y=&Lerp_Float(A=0.5, B=1, Alpha=$User.uPhase)
       Z=&Lerp_Float(A=0.5, B=1, Alpha=$User.uPhase)
      )
     )
     B=&VectorFromFloat(Value=&FloatFromCurve(
       FloatCurve=DataInterface() {propertyValues {Curve {keys [
           {interpMode=RCIM_Cubic, tangentMode=RCTM_User, time=0.3, value=1}
           {arriveTangent=-3.2987, interpMode=RCIM_Cubic, leaveTangent=-3.2987, tangentMode=RCTM_User, time=1, value=0.2}
          ]}}}
       CurveIndex=$Emitter.NormalizedLoopAge
      ))
    )
   ) {isSetParameters=true}
   /SetVariables(
    Particles.Color_Thruster:LinearColor=&ScaleLinearColorByCurve(
     InitialValue=$Particles.Initial.Color_Thruster
     ScaleCurve=DataInterface() {propertyValues {
       AlphaCurve {keys [<at 0 0.05 RCIM_Cubic>]}
       BlueCurve {keys [<at 0.0203 2>, <at 0.0654 1>, {arriveTangent=-15.6046, time=0.133, value=1}, <at 0.3627 1>, <at 0.6962 1>]}
       GreenCurve {keys [<at 0.0203 0.1387>, <at 0.0654 0.7124>, {arriveTangent=-5.6465, time=0.133, value=0.2645}, <at 0.3627 0.0297>, <at 0.6962 0.1856>]}
       RedCurve {keys [<at 0.0203 0>, <at 0.0654 0.3103>, {arriveTangent=-0.716, time=0.133, value=0.4409}, <at 0.3627 0.0363>, <at 0.6962 0>]}
      }}
     CurveIndex=$Particles.NormalizedAge
    )
    Particles.Size_Thruster:Vector3f=&Multiply_Vector(
     A=$Particles.Initial.Size_Thruster
     B=&Multiply_Vector(
      A=&VectorFromCurve(
       VectorCurve=DataInterface() {propertyValues {
         XCurve {keys [<at 0 1>,<at 1 1>]}
         YCurve {keys [<at 0 1>,<at 1 1>]}
         ZCurve {keys [<at 0.9 1 RCIM_Cubic>, {arriveTangent=-4.873, interpMode=RCIM_Cubic, leaveTangent=-4.873, tangentMode=RCTM_User, time=1, value=0}]}
        }}
       Scale Curve=&Lerp_Vector(A=<0.5 0.5 1.5>, B=<0.75 0.75 3>, Alpha=$User.uPhase)
      )
      B=&MakeVector(
       X=1
       Y=1
       Z=&FloatFromCurve(FloatCurve=DataInterface() {propertyValues {Curve {keys [{interpMode=RCIM_Cubic, tangentMode=RCTM_User, time=0, value=1}, {interpMode=RCIM_Cubic, tangentMode=RCTM_User, time=1, value=0}]}}})
      )
     )
    )
   ) {isSetParameters=true}
   &ParticleState()
  ]
  rendererClasses [&NiagaraMeshRendererProperties]
  renderers [
   {rendererClass=&NiagaraMeshRendererProperties, rendererIndex=0}
   {rendererClass=&NiagaraMeshRendererProperties, rendererIndex=1}
  ]
 }}
summary {
 systemName=NS_S28_Narrative_Booster_Holo
 emitters [{emitterName=Combined_Rate, rendererClasses [&NiagaraMeshRendererProperties]}]
 userVariables [
  User.FlareFlicker {type=NiagaraFloat, defaultValue=1, description=""}
  User.uDuration {type=NiagaraFloat, defaultValue=1.5, description=""}
  User.uPhase {type=NiagaraFloat, defaultValue=1, description=""}
 ]
}
