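# Niagara system for a Season 28 narrative quest actor, named as a holo popcorn effect, showing small shuffling kernel sprites.
# One emitter spawns short-lived particles intermittently via spawn probability inside a small sphere, tossing them upward in one of two pale yellow tints.
# Gravity and low-bounce collision let them fall and settle.
# When a particle comes to rest, an update step zeroes its first dynamic material parameter, otherwise preserving the current value.
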NiagaraSystem /BattlePassS28_Quests/Items/QuestActors/S28_Narrative/NS_S28_Narrative_Popcorn_03_Holo
dependencies {usedDataInterfaces [], usedDynamicInputs [], usedModules [&SystemState]}
emitters {Corny_Shuffling {
  emitterUpdateScript [
   &EmitterState()
   &SpawnRate(Spawn Probability=0.35, SpawnRate=25, Use Spawn Probability=true)
  ]
  particleSpawnScript [
   /Niagara/Modules/Spawn/Initialization/InitializeParticle(
    Color=&LerpLinearColors(
     StartColor=<rgba 0.95 0.7323 0.4275 1>
     EndColor=<rgba 0.95 0.8313 0.665 1>
     LerpFactor=/Niagara/DynamicInputs/Integer/V2/Round_Float(Float:NiagaraFloat=&RandomRangeFloat(Minimum=0, Maximum=1))
    )
    Lifetime=0.4
    Module.Write Color=true
    Module.Write Lifetime=true
    Module.Write Mass=true
    Module.Write Material Random=true
    Module.Write Position=false
    Module.Write Scale=false
    Module.Write SpriteRotation=false
    Module.Write SpriteSize=true
    Sprite Size=&Vector2DFromFloat(Value=&RandomRangeFloat(Minimum=4, Maximum=5))
   )
   &DynamicMaterialParameters(Module.Param0WriteEnabled=true, Module.Param1WriteEnabled=true, Module.Param2WriteEnabled=true, Module.Param3WriteEnabled=true)
   &AddVelocity(Coordinate Space=::World, Velocity=&RandomRangeVector(Minimum=<-15 -15 10>, Maximum=<15 15 30>))
   &SphereLocation(
    MaskSpawnGroup=false
    Override Seed=false
    Sphere Radius=5
    Surface Only=false
    UseNonUniformScale=false
    UseOffset=false
    UseSphereOrientation=false
   )
  ]
  particleUpdateScript [
   &ParticleState()
   &GravityForce(Gravity=<0 0 -350>)
   &Collision(
    Age Colliding Particles=false
    Friction=0.35
    Module.Control Rest State Via Surface Normal=true
    Module.EnableMaxCollisionCount=false
    Module.Randomize Collision Normal Vector=false
    Restitution=0.3
   )
   &DynamicMaterialParameters(
    Index 0 Param 1=&MakeCustomFloatFromBool(Bool=$Transient.CollisionIsResting, True Float=0, False Float=HlslExpression() {hlslExpression=Particles.DynamicMaterialParameter.x})
    Module.Param0WriteEnabled=true
    Module.Param1WriteEnabled=false
    Module.Param2WriteEnabled=false
    Module.Param3WriteEnabled=false
   ) {module=DynamicMaterialParameters001}
   &SolveForcesAndVelocity(Clamp Velocity=false, Limit Acceleration=false)
  ]
  rendererClasses [&NiagaraSpriteRendererProperties]
  renderers [{rendererClass=&NiagaraSpriteRendererProperties}]
 }}
summary {
 systemName=NS_S28_Narrative_Popcorn_03_Holo
 emitters [{emitterName=Corny_Shuffling, rendererClasses [&NiagaraSpriteRendererProperties]}]
 userVariables []
}
